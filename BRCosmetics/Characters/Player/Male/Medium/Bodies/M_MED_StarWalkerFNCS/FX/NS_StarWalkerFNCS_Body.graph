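# Niagara body effect for the StarWalkerFNCS character cosmetic, producing small drifting fragments with orange glints.
# TinyFrag continuously spawns small bright white meshes at random vertical offsets that rotate, drift and fade in and out through size and alpha curves.
# Glints uses a particle attribute reader to spawn sprites from TinyFrag particles, rejecting older sources, with a curve that spikes their size late in life.
# The Flare emitter has no renderer.
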NiagaraSystem /BRCosmetics/Characters/Player/Male/Medium/Bodies/M_MED_StarWalkerFNCS/FX/NS_StarWalkerFNCS_Body
dependencies {
 usedDataInterfaces [&NiagaraDataInterfaceCurve, &NiagaraDataInterfaceParticleRead]
 usedDynamicInputs []
 usedModules [&SystemState]
}
emitters {
 Flare {rendererClasses [], renderers []}
 Glints {
  emitterUpdateScript [
   &EmitterState()
   &SpawnParticlesFromOtherEmitter(
    Attribute Reader=DataInterface() {propertyValues {EmitterBinding {bindingMode=Other, emitterName=TinyFrag}}}
    Spawn Rate=1
    Use Maximum Age=false
    Use Minimum Age=false
   )
  ]
  particleSpawnScript [
   &InitializeParticle(
    Color=<rgba 5 1.6667 0 1>
    Lifetime Max=1.2
    Lifetime Min=0.12
    Lifetime Mode=::NewEnumerator1
    Sprite Size Max=<40 30>
    Sprite Size Min=<20 10>
    Sprite Size Mode=::NewEnumerator2
    UsePositionOffset=false
    Write Material Random=true
   )
   &SampleParticlesFromOtherEmitter(
    Maximum Sampled Particle Age=0.9
    Module.MaxSampledAgeCutoff=true
    Module.MinSampledAgeCutoff=false
    Reject Particle By Sampled Age=true
    Reject Particle on Sampled Particle's First Frame=true
    Velocity Sampling=::NewEnumerator0
   )
  ]
  particleUpdateScript [
   &ParticleState()
   &ScaleSpriteSize(Uniform Curve Sprite Scale=DataInterface() {propertyValues {Curve {keys [
        <at 0 0 RCIM_Cubic>
        {
         arriveTangent=1.4286
         arriveTangentWeight=0.3342
         interpMode=RCIM_Cubic
         leaveTangent=1.4286
         leaveTangentWeight=0.0727
         time=0.575
         value=0.35
        }
        {arriveTangent=2.685, interpMode=RCIM_Cubic, leaveTangent=2.685, tangentMode=RCTM_User, time=0.7, value=1}
        <at 1 0 RCIM_Cubic>
       ]}}})
   &SolveForcesAndVelocity(Clamp Velocity=false, Limit Acceleration=false)
  ]
  rendererClasses [&NiagaraSpriteRendererProperties]
  renderers [{rendererClass=&NiagaraSpriteRendererProperties}]
 }
 TinyFrag {
  emitterUpdateScript [
   &EmitterState(
    Life Cycle Mode=::NewEnumerator1
    Loop Duration=&RandomRangeFloat(Minimum=0.1, Maximum=0.2)
    Recalculate Duration Each Loop=true
    UseLoopDelay=false
   )
   &SpawnRate(SpawnRate=5, Use Spawn Probability=false)
  ]
  particleSpawnScript [
   &InitializeParticle(
    Color=<rgba 10 10 10 1>
    Lifetime Mode=::NewEnumerator1
    Mesh Scale Mode=::NewEnumerator1
    Mesh Uniform Scale Max=0.4
    Mesh Uniform Scale Min=0.1
    Position Offset=&RandomRangeVector(Minimum=<0 0 -15>, Maximum=<0 0 15>)
    UsePositionOffset=true
    Write Material Random=true
   )
   &InitialMeshOrientation(Use Rotation Vector=false)
  ]
  particleUpdateScript [
   &ParticleState()
   &UpdateMeshOrientation(Rotation Vector=&RandomRangeVector(Minimum=<-0.15 -0.15 -0.15>, Maximum=<0.15 0.15 0.15>))
   &ScaleMeshSize(Scale Factor=&VectorFromFloat(Value=&FloatFromCurve(FloatCurve=DataInterface() {propertyValues {Curve {keys [
          {interpMode=RCIM_Cubic, leaveTangentWeight=0.0667, tangentMode=RCTM_User, time=0, value=0}
          {arriveTangentWeight=0.1067, interpMode=RCIM_Cubic, leaveTangentWeight=0.3202, tangentMode=RCTM_User, time=0.2, value=1}
          {arriveTangentWeight=0.3202, interpMode=RCIM_Cubic, leaveTangentWeight=0.1067, tangentMode=RCTM_User, time=0.8, value=1}
          {arriveTangentWeight=0.0667, interpMode=RCIM_Cubic, tangentMode=RCTM_User, time=1, value=0}
         ]}}})))
   &ScaleColor(
    Scale Alpha=&FloatFromCurve(FloatCurve=DataInterface() {propertyValues {Curve {keys [
         {interpMode=RCIM_Cubic, leaveTangentWeight=0.0667, tangentMode=RCTM_User, time=0, value=0}
         {arriveTangentWeight=0.1163, interpMode=RCIM_Cubic, leaveTangentWeight=0.2906, tangentMode=RCTM_User, time=0.2, value=1}
         {arriveTangentWeight=0.2668, interpMode=RCIM_Cubic, leaveTangentWeight=0.1601, tangentMode=RCTM_User, time=0.7, value=1}
         {arriveTangentWeight=0.1, interpMode=RCIM_Cubic, tangentMode=RCTM_User, time=1, value=0}
        ]}}})
    ScaleA=true
    ScaleRGB=true
   )
   &AccelerationForce(Acceleration=&RandomRangeVector(Minimum=<-5 5 2>, Maximum=<5 5 2>))
   &SolveForcesAndVelocity(Clamp Velocity=false, Limit Acceleration=false)
  ]
  rendererClasses [&NiagaraMeshRendererProperties]
  renderers [{rendererClass=&NiagaraMeshRendererProperties, rendererIndex=0}]
 }
}
summary {
 systemName=NS_StarWalkerFNCS_Body
 emitters [
  {emitterName=Flare, rendererClasses []}
  {emitterName=TinyFrag, rendererClasses [&NiagaraMeshRendererProperties]}
  {emitterName=Glints, rendererClasses [&NiagaraSpriteRendererProperties]}
 ]
 userVariables []
}
